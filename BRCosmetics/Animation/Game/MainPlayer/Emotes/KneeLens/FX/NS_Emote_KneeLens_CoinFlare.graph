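# Niagara system for the coin flare effect in the KneeLens emote, built from a single DirectionalBurst sprite emitter.
# A probability-gated instantaneous burst spawns short-lived sprites with random lifetime, uniform size and material random.
# Spawn color comes from a time-of-day function, warm gold for morning, day and evening and pink at night.
# A curve fades alpha to near zero over each particle's life.
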NiagaraSystem /BRCosmetics/Animation/Game/MainPlayer/Emotes/KneeLens/FX/NS_Emote_KneeLens_CoinFlare
dependencies {usedDataInterfaces [&NiagaraDataInterfaceCurve], usedDynamicInputs [], usedModules [&SystemState]}
emitters {DirectionalBurst {
  emitterUpdateScript [
   &EmitterState(Life Cycle Mode=::NewEnumerator1, Loop Behavior=::NewEnumerator1, Loop Duration=1, UseLoopDelay=false)
   &SpawnBurst_Instantaneous(Use Spawn Probability=true, UseLoopCountLimit=false)
  ]
  particleSpawnScript [&InitializeParticle(
    Color=&FNTimeOfDayColor(
     Night Color=<rgba 1 0.502 0.607 1>
     Morning Color=<rgba 1 0.6735 0.2372 1>
     Day Color=<rgba 1 0.6735 0.2372 1>
     Evening Color=<rgba 1 0.6735 0.2372 1>
    )
    Lifetime=&RandomRangeFloat(Minimum=0.25, Maximum=0.35)
    Mass Max=2
    Mass Mode=::NewEnumerator3
    Sprite Size Mode=::NewEnumerator3
    Uniform Sprite Size=&RandomRangeFloat(Minimum=25, Maximum=37)
    UsePositionOffset=true
    Write Material Random=true
   )]
  particleUpdateScript [
   &ParticleState()
   &ScaleColor(
    Scale Alpha=&FloatFromCurve(FloatCurve=DataInterface() {propertyValues {Curve {keys [<at -0.0046 1.0113 RCIM_Cubic>,<at 1.0011 0.002 RCIM_Cubic>]}}})
    ScaleA=true
    ScaleRGB=true
   )
   &SolveForcesAndVelocity(Clamp Velocity=false, Limit Acceleration=false)
  ]
  rendererClasses [&NiagaraSpriteRendererProperties]
  renderers [{rendererClass=&NiagaraSpriteRendererProperties}]
 }}
summary {
 systemName=NS_Emote_KneeLens_CoinFlare
 emitters [{emitterName=DirectionalBurst, rendererClasses [&NiagaraSpriteRendererProperties]}]
 userVariables []
}
